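# Niagara swing-trail effect for the TapDance one-handed pickaxe, pairing an animation-driven ribbon with orange lightning.
# FN_AnimTrail_V5 draws a pale red ribbon that narrows with particle age and gets shorter-lived across the emitter loop.
# Lightning scatters mesh particles along the weapon's skeletal mesh by random height, computing flickering light attributes despite having only a mesh renderer.
# The disabled Mesh_Trails emitter scales trail length and alpha from sampled socket velocity.
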NiagaraSystem /BRCosmetics/Weapons/FORT_Melee/Pickaxe_TapDance_Female_1H/Effects/NS_Pickaxe_TapDance_AnimTrail
dependencies {
 usedDataInterfaces [&NiagaraDataInterfaceCurve, &NiagaraDataInterfaceVectorCurve]
 usedDynamicInputs []
 usedModules [
  &self:SystemSpawnScript.SystemScriptSource.SystemScriptGraph.NiagaraNodeAssignment_0.SetVariables_#1
  &self:SystemSpawnScript.SystemScriptSource.SystemScriptGraph.NiagaraNodeAssignment_1.SetVariables_#2
  &SystemState
 ]
}
emitters {
 FN_AnimTrail_V5 {
  emitterUpdateScript [
   &EmitterState()
   &FN_SpawnAnimTrail_v2()
  ]
  particleSpawnScript [
   /Game/Effects/Niagara/Modules/AnimTrail/FN_InitializeAnimTrail/FN_InitializeAnimTrailBoneIndecies(Second Bone or Socket Index=1)
   /Game/Effects/Niagara/Modules/AnimTrail/FN_InitializeAnimTrail/FN_InitializeAnimTrail(
    AnimTrailWidth=2
    Color=<rgba 1 0.364 0.364 0.5>
    Lifetime:NiagaraFloat=&Multiply_Float(A=0.15, B=&Lerp_Float(A=1, B=0.2, Alpha=$Emitter.NormalizedLoopAge))
   )
   &AddVelocity(Velocity=$System.TestVelocity)
  ]
  particleUpdateScript [
   &SolveForcesAndVelocity(Clamp Velocity=false, Limit Acceleration=false)
   &FN_UpdateAnimTrail_v2()
   &ScaleRibbonWidth(Ribbon Width Scale=&Lerp_Float(A=2, B=0, Alpha=$Particles.NormalizedAge))
  ]
  rendererClasses [&NiagaraRibbonRendererProperties]
  renderers [{rendererClass=&NiagaraRibbonRendererProperties}]
 }
 Lightning {
  emitterUpdateScript [
   &EmitterState()
   &SpawnRate(SpawnRate=50, Use Spawn Probability=false)
  ]
  particleSpawnScript [
   /SetVariables(Particles.NormalizedHeight:NiagaraFloat=&RandomRangeFloat(Minimum=0, Maximum=1)) {isSetParameters=true}
   /Niagara/Modules/Spawn/Initialization/InitializeParticle(
    Color=&MultiplyLinearColor(
     Linear Color=&RandomRangeLinearColor(Minimum=<rgba 5 0.2633 0.1 1>, Maximum=<rgba 10 0.3333 0 1>)
     Scale Factor=&MakeVector4_FromFloat(Float=&RandomRangeFloat(Minimum=0.04, Maximum=&Lerp_Float(A=1, B=0.25, Alpha=$Particles.NormalizedHeight)))
    )
    Lifetime=&RandomRangeFloat(Minimum=0.2, Maximum=0.4)
    Mesh Scale=&Lerp_Vector(
     A=&RandomRangeVector(Minimum=<0.025 0.01 0.01>, Maximum=<0.055 0.05 0.05>)
     B=&RandomRangeVector(Minimum=<0.01 0.02 0.02>, Maximum=<0.02 0.02 0.02>)
     Alpha=$Particles.NormalizedHeight
    )
    Module.Write Color=true
    Module.Write Lifetime=true
    Module.Write Mass=false
    Module.Write Material Random=true
    Module.Write Position=true
    Module.Write Scale=true
    Module.Write SpriteRotation=false
    Module.Write SpriteSize=true
    Position=&Lerp_Vector(A=<0 0 0>, B=<0 0 32>, Alpha=$Particles.NormalizedHeight)
    Sprite Size=&Vector2DFromFloat(Value=4)
   )
   /Niagara/Modules/Spawn/Orientation/V2/InitialMeshOrientation(
    Orientation Vector=<0 0 1>
    Rotation=&UniformRangedVector(Minimum=<-360 0 0>, Maximum=<360 0 0>)
    Use Orientation Vector=true
    Use Rotation Vector=true
   )
   &DynamicMaterialParameters(
    Index 0 Param 1=&RandomRangeFloat(Minimum=0.5, Maximum=1.25)
    Index 0 Param 2=&Lerp_Float(A=&RandomRangeFloat(Minimum=0.4, Maximum=0.75), B=0.2, Alpha=$Particles.NormalizedHeight)
    Index 0 Param 3=1
    Module.Param0WriteEnabled=true
    Module.Param1WriteEnabled=true
    Module.Param2WriteEnabled=true
    Module.Param3WriteEnabled=true
   )
   &SystemLocation(MaskSpawnGroup=false, Offset=&Lerp_Vector(A=<0 0 0>, B=<0 0 32>, Alpha=$Particles.NormalizedHeight))
  ]
  particleUpdateScript [
   &ParticleState()
   &SkeletalMeshLocation(
    Inherit Velocity=false
    MaskSpawnGroup=false
    Module.Override Randomness=false
    Sampled Position Offset=&Lerp_Vector(A=<0 0 0>, B=<0 0 50>, Alpha=$Particles.NormalizedHeight)
    Sampled Position Offset Coordinate Space:ENiagaraMeshTransforms=::NewEnumerator3
    Skeletal Mesh=$User.SkeletalMesh
   )
   &ScaleMeshSize(Scale Factor=&VectorFromCurve(VectorCurve=DataInterface() {propertyValues {
       XCurve {keys [
         {arriveTangent=1.4898, interpMode=RCIM_Cubic, leaveTangent=1.4898, tangentMode=RCTM_User, time=-0.0011, value=0.5}
         {arriveTangent=0.2007, interpMode=RCIM_Cubic, leaveTangent=0.2007, time=0.5, value=0.95}
         {arriveTangent=-1.4081, interpMode=RCIM_Cubic, leaveTangent=-1.4081, tangentMode=RCTM_User, time=0.9954, value=0.7}
        ]}
       YCurve {keys [<at 0 0.3>,<at 1 1>]}
       ZCurve {keys [<at 0 0.3>,<at 1 1>]}
      }}))
   &SolveForcesAndVelocity(Clamp Velocity=false, Limit Acceleration=false)
   &FN_Light_v3(
    Color=$Particles.Color
    Create Color Attribute=true
    Create Exponent Attribute=true
    Create Radius Attribute=true
    Exponent=2
    Radius:NiagaraFloat=/Niagara/DynamicInputs/Math/LerpMultipleFloats(
     Float Count:ENiagara_MultipleLerpCount=::NewEnumerator0
     1:NiagaraFloat=3
     2:NiagaraFloat=8
     3:NiagaraFloat=0
     4:NiagaraFloat=?
     5:NiagaraFloat=?
     Alpha:NiagaraFloat=$Particles.NormalizedAge
    )
    Scale Color:NiagaraFloat=&FloatFromCurve(
     FloatCurve=DataInterface() {propertyValues {Curve {keys [<at 0 1>,<at 0.1656 0.2263>,<at 0.2177 1>,<at 0.3102 0.538>,<at 0.4973 0.8>,<at 0.6182 0.5013>,<at 1 0>]}}}
     Scale Curve=&Multiply_Float(
      A=25
      B=&Add_Float(
       A=&MakeFloatFromVector4(Vector 4=$Particles.DynamicMaterialParameter, Channel=::NewEnumerator0)
       B=-0.25
      )
     )
    )
   )
   /SetVariables(Particles.PositionLightOffset:Vector3f=&Add_Vector(
     A=Linked() {linkedVariable {name=Particles.Position, type=NiagaraPosition}}
     B=&Add_Vector(
      A=&Multiply_VectorByFloat(
       Vector=<0 0 1>
       Float=&Multiply_Float(A=&MakeFloatFromVector(Vector=$Particles.Scale, Channel=::NewEnumerator0), B=400)
      )
      B=&Multiply_Vector(
       A=<1 1 0>
       B=&Add_Vector(
        A=&Lerp_Vector(A=<2 2 0>, B=<0 0 0>, Alpha=$Particles.NormalizedAge)
        B=&Multiply_VectorByFloat(
         Vector=&RandomVector(Vector Scale=4)
         Float=&LerpMultipleFloats(
          Float Count:ENiagara_MultipleLerpCount=::NewEnumerator0
          1:NiagaraFloat=0.5
          2:NiagaraFloat=1
          3:NiagaraFloat=0.5
          4:NiagaraFloat=?
          5:NiagaraFloat=?
          Alpha:NiagaraFloat=$Particles.NormalizedAge
         )
        )
       )
      )
     )
    )) {isSetParameters=true}
   /SetVariables(
    Particles.PositionLight01:Vector3f=$Particles.PositionLightOffset
    Particles.PositionLight02:Vector3f=&Multiply_Vector(A=$Particles.PositionLightOffset, B=<-1 -1 1>)
   ) {isSetParameters=true}
  ]
  rendererClasses [&NiagaraMeshRendererProperties]
  renderers [{rendererClass=&NiagaraMeshRendererProperties, rendererIndex=0}]
 }
 Mesh_Trails {
  bEnabled=false
  emitterUpdateScript [
   &EmitterState()
   &SpawnBurst_Instantaneous(Spawn Count=1, Use Spawn Probability=false) {module=SpawnBurst_Instantaneous001}
  ]
  particleSpawnScript [
   &InitializeParticle(
    Lifetime=0.25
    Mesh Scale=&VectorFromFloat(Value=5)
    Module.Write Color=true
    Module.Write Lifetime=true
    Module.Write Mass=false
    Module.Write Material Random=true
    Module.Write Position=false
    Module.Write Scale=true
    Module.Write SpriteRotation=false
    Module.Write SpriteSize=false
   )
   &self:Mesh_Trail_Setup(Trail Length=100, Trail Width=50) {module=Mesh_Trail_Setup}
   &InitialMeshOrientation(Orientation Axis:Vector3f=<0 1 0>, Orientation Vector=<0 0 -1>, Use Orientation Vector=true, Use Rotation Vector=false)
  ]
  particleUpdateScript [
   &ParticleState(Kill Particles When Lifetime Has Elapsed=false, Let Infinitely Lived Particles Die When Emitter Deactivates=true)
   &SkeletalMeshLocation(
    Apply Sampled Orientation to Mesh Particles=true
    Bone Sampling Mode=::NewEnumerator4
    Inherit Velocity=false
    Interpolate Mesh Sampling=false
    MaskSpawnGroup=false
    Sampled Position Offset=<0 0 25>
    Sampled Position Offset Coordinate Space:ENiagaraMeshTransforms=::NewEnumerator3
    Skeletal Mesh=$User.SkeletalMesh
    Write Sampled Mesh Attrributes to Particle Attributes=true
   ) {module=SkeletalMeshLocation001}
   /SetVariables(Particles.FakeVelocity:Vector3f=<1101 365 47>) {enabled=false, isSetParameters=true}
   /SetVariables(
    Particles.NormalishSocketVelocity:Vector3f=/Niagara/DynamicInputs/Helpers/SafeNormalizeVector(
     Vector To Normalize:Vector3f=&Multiply_Vector(A=$Particles.SkeletalMeshLocation001.SampledVelocity, B=<0.1 1 -1>)
     Fallback Vector:Vector3f=<0 0 0>
    )
    Particles.NormalishSocketVelocityLength:NiagaraFloat=&Multiply_Float(
     A=&self:InterpolateFloat(Interp Input:NiagaraFloat=&ClampFloat(Float=&Divide_Float(
        A=&Add_Float(A=/Niagara/DynamicInputs/TypeConversions/VectorLengthSafe(Vector:Vector3f=$Particles.SkeletalMeshLocation001.SampledVelocity), B=-200)
        B=1000
       )))
     B=&FloatFromCurve(FloatCurve=DataInterface() {propertyValues {Curve {keys [<at 0.8 1>,<at 1 0>]}}})
    )
   ) {enabled=false, isSetParameters=true}
   /SetVariables(
    Particles.NormalishSocketVelocity:Vector3f=&Multiply_Vector(
     A=$Particles.SkeletalMeshLocation001.SampledVelocity
     B=&RotateVectorByQuaternion(Vector To Rotate=<-0.001 0.001 0>, Quaternion=$Particles.SkeletalMeshLocation001.SampledRotation)
    )
    Particles.NormalishSocketVelocityLength:NiagaraFloat=&Multiply_Float(
     A=&self:InterpolateFloat(Interp Input:NiagaraFloat=&ClampFloat(Float=&Divide_Float(
        A=&Add_Float(A=&VectorLengthSafe(Vector:Vector3f=$Particles.SkeletalMeshLocation001.SampledVelocity), B=-200)
        B=1000
       )))
     B=&FloatFromCurve(FloatCurve=DataInterface() {propertyValues {Curve {keys [<at 0.75 1>,<at 1 0>]}}})
    )
   ) {isSetParameters=true}
   &self:Exhaust_to_DynaMat(ExhaustVector:Vector3f=&TransformVector(Vector=$Particles.NormalishSocketVelocity, Source Space=::World, Destination Space=::Local)) {enabled=false, module=Exhaust_to_DynaMat}
   &self:Exhaust_to_DynaMat(
    ExhaustVector:Vector3f=&Multiply_VectorByFloat(Vector=$Particles.NormalishSocketVelocity, Float=10)
    SpeedMult:NiagaraFloat=$Particles.NormalishSocketVelocityLength
   ) {module=Exhaust_to_DynaMat001}
   &self:Mesh_Trail_Length_Update(Dynamic Length Scale:NiagaraFloat=$Particles.NormalishSocketVelocityLength) {module=Mesh_Trail_Length_Update}
   &ScaleColor(Scale Alpha=$Particles.NormalishSocketVelocityLength, ScaleA=true, ScaleRGB=false)
   /SetVariables() {isSetParameters=true}
  ]
  rendererClasses [&NiagaraMeshRendererProperties, &NiagaraSpriteRendererProperties]
  renderers [
   {rendererClass=&NiagaraMeshRendererProperties, rendererIndex=0}
   {rendererClass=&NiagaraMeshRendererProperties, rendererIndex=1}
   {rendererClass=&NiagaraSpriteRendererProperties, rendererIndex=2}
  ]
 }
}
summary {
 systemName=NS_Pickaxe_TapDance_AnimTrail
 emitters [
  {
   bEnabled=false
   emitterName=Mesh_Trails
   rendererClasses [&NiagaraMeshRendererProperties, &NiagaraSpriteRendererProperties]
  }
  {emitterName=FN_AnimTrail_V5, rendererClasses [&NiagaraRibbonRendererProperties]}
  {emitterName=Lightning, rendererClasses [&NiagaraMeshRendererProperties]}
 ]
 userVariables [
  User.bIsFrontEndPreview {type=NiagaraBool}
  User.BPOpacityMult {type=NiagaraFloat, defaultValue=0, description=""}
  User.CurrentExhaust {type=Vector3f, description="", defaultValue {struct=Vector3f, value=<0 0 0>}}
  User.PlayerSpeed {type=NiagaraFloat, defaultValue=0, description=""}
  User.RibbonIDOffset {type=NiagaraInt32}
  User.SkeletalMesh {
   type=NiagaraDataInterfaceSkeletalMesh
   defaultValue {
    struct=NiagaraExt_VariableValue_DataInterface
    value {dataInterface=&self:NiagaraDataInterfaceSkeletalMesh_2}
   }
  }
  User.TrailLengthScale {type=NiagaraFloat, defaultValue=0, description=""}
 ]
}
